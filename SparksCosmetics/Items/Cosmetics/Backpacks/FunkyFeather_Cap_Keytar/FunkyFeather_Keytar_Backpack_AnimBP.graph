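# Anim Blueprint for the FunkyFeather keytar back bling, positioning the Generic Keytar skeleton as a backpack.
# The AnimGraph applies two Modify Bone steps to the offset bone: a replace for base position, rotation and scale, then an additive adjustment.
# Each update, clamped spine and right upper-arm yaw from the owner's mesh set backpackRotate and backpackTranslate.
# No visible AnimGraph link reads those variables.
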AnimBlueprint /SparksCosmetics/Items/Cosmetics/Backpacks/FunkyFeather_Cap_Keytar/FunkyFeather_Keytar_Backpack_AnimBP
target_skeleton=/SparksCharacterCommon/Instrument/Keytar/Generic/Meshes/Generic_Keytar_Skeleton
graphs [
 AnimGraph {nodes [
   n1 = "Transform (Modify) Bone - Bone: offset"!(
    ComponentPose=Local To Component(LocalPose=Input Pose())
    bAlphaBoolEnabled:bool=True
    Translation="5,6,3"
    Rotation="5,75,0"
    Scale="0.9,0.9,0.9"
   ) -> (Pose) {
    alphaInputType=Bool
    rotationMode=BMM_Replace
    scaleMode=BMM_Replace
    translationMode=BMM_Replace
    alphaBoolBlend {blendOption=Linear}
    boneToModify {boneName=offset}
    rotation {pitch=5, yaw=75}
    scale=<0.9 0.9 0.9>
    translation=<5 6 3>
   }
   n2 = "Transform (Modify) Bone - Bone: offset"!(ComponentPose=@n1, Alpha:float=1, Translation="0,5,-5", Rotation="0,0,0") -> (Pose) {rotationMode=BMM_Additive, scaleMode=BMM_Additive, translationMode=BMM_Additive, boneToModify {boneName=offset}, translation {y=5, z=-5}}
   Output Pose(Result=Component To Local(ComponentPose=@n2))
  ]}
 EventGraph {nodes [
   Event BlueprintUpdateAnimation()
   n1 = GetOwnerSkeletalMeshComponent()
   n2 = BreakRotator(InRot=BreakTransform(InTransform=GetDeltaTransformFromRefPose(self=@n1, BoneName=spine_03, BaseName=spine_05)).Rotation)
   n3 = MakeRotator(Roll=MapRangeClamped(Value=@n2.Yaw, InRangeA=0, InRangeB=15, OutRangeB=-10), Yaw=-7)
   > Set backpackRotate(backpackRotate:Rotator=@n3) -> (Output_Get:Rotator)
   n4 = float + float(A=MapRangeClamped(Value=@n2.Yaw, InRangeA=0, InRangeB=15, OutRangeB=5), B=5)
   n5 = GetOwnerSkeletalMeshComponent()
   n6 = BreakRotator(InRot=BreakTransform(InTransform=GetDeltaTransformFromRefPose(self=@n5, BoneName=spine_05, BaseName=upperarm_r)).Rotation)
   n7 = MapRangeClamped(Value=@n6.Yaw, InRangeA=0, InRangeB=-50)
   n8 = float * float(A=@n7, B=5) {is=title K2Node_PromotableOperator}
   n9 = float * float(A=@n7, B=-5)
   > Set backpackTranslate(backpackTranslate:Vector=vector + vector(A=MakeVector(X=@n9, Y=@n8), B=MakeVector(X=-5, Y=@n4, Z=-5))) -> (Output_Get:Vector)
   MapRangeClamped(Value=@n6.Yaw, InRangeA=5, InRangeB=-15, OutRangeB=15)
   Reroute Node(InputPin:Rotator=@n3) -> (OutputPin:Rotator)
  ]}
]
variables [
 backpackRotate {type=Rotator}
 backpackTranslate {type=Vector}
]
